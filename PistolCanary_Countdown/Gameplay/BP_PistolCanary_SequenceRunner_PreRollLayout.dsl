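; Sequence runner for the PistolCanary Countdown pre-roll, setting the world's layout and portal audio for three event days.
; On authority, match-ready and calendar sync trigger loading for each active day.
; Day loaders chain back to Day 1's reset, switch data layers after delays, and briefly play each day's sequence through a spawned replicated actor.
; Cheat events preview days and turn on lights, or pre-load and play the main sequence.
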(event EventBeginPlay
  ;[0.1326 0.15 0.01508 0.5] Calendar
  (Networking|SwitchHasAuthority
    (:Authority
      (if (Variables|Default|GetFollowCalendarEvents)
        (Class|FortAsyncActionAuthorityMatchReady|MatchReadyAsync
          (:OnReady
            (MCPContext|InitialCalendarSyncCompleted
              (:OnSyncComplete
                (Utilities|FlowControl|Sequence
                  (:then_0
                    (bind _biseventactive (CalendarEvents|GetCalendarEventInformation (Variables|Default|GetCalendarEventToWatch_Day1)))
                    (if _biseventactive
                      (CallFunction|LoadDataLayersforDay1)
                      (CallFunction|SetPortalAudio 0)
                      (else
                        (Fortnite|FortShippingLog self "Day1 Calendar Event is Inactive"))))
                  (:then_1
                    (bind _biseventactive_2 (CalendarEvents|GetCalendarEventInformation (Variables|Default|GetCalendarEventToWatch_Day2)))
                    (if _biseventactive_2
                      (CallFunction|LoadDataLayersforDay2)
                      (CallFunction|SetPortalAudio 1)
                      (else
                        (Fortnite|FortShippingLog self "Day2 Calendar Event is Inactive"))))
                  (:then_2
                    (bind _biseventactive_3 (CalendarEvents|GetCalendarEventInformation (Variables|Default|GetCalendarEventToWatch_Day3)))
                    (if _biseventactive_3
                      (CallFunction|LoadDataLayersforDay3)
                      (CallFunction|SetPortalAudio 2)
                      (else
                        (Fortnite|FortShippingLog self "Day3 Calendar Event is Inactive")))))))))))))

(event Custom|PC_PreviewDay1
  ;[0 0 0] Console Command / Load and Play
  (Variables|Default|SetRunFromCheat true)
  (CallFunction|LoadDataLayersforDay1)
  (CallFunction|SetPortalAudio 0)
  (Utilities|FlowControl|Delay 7.77)
  (Development|ExecuteConsoleCommand "cheat ke * PC_TurnOnLights"))

(event Custom|PC_Day1
  ;[0 0 0] Console Command / Load and Play
  (Variables|Default|SetRunFromCheat true)
  (CallFunction|LoadDataLayersforDay1)
  (CallFunction|SetPortalAudio 0)
  (Utilities|FlowControl|Delay 7.77)
  (Development|ExecuteConsoleCommand "cheat ke * PC_TurnOnLights"))

(event Custom|PC_PreviewDay2
  ;[0 0 0] Console Command / Load and Play
  (Variables|Default|SetRunFromCheat true)
  (CallFunction|LoadDataLayersforDay2)
  (CallFunction|SetPortalAudio 1)
  (Utilities|FlowControl|Delay 7.77)
  (Development|ExecuteConsoleCommand "cheat ke * PC_TurnOnLights"))

(event Custom|PC_Day2
  ;[0 0 0] Console Command / Load and Play
  (Variables|Default|SetRunFromCheat true)
  (CallFunction|LoadDataLayersforDay2)
  (CallFunction|SetPortalAudio 1)
  (Utilities|FlowControl|Delay 7.77)
  (Development|ExecuteConsoleCommand "cheat ke * PC_TurnOnLights"))

(event Custom|PC_PreviewDay3
  ;[0 0 0] Console Command / Load and Play
  (Variables|Default|SetRunFromCheat true)
  (CallFunction|LoadDataLayersforDay3)
  (CallFunction|SetPortalAudio 2)
  (Utilities|FlowControl|Delay 7.77)
  (Development|ExecuteConsoleCommand "cheat ke * PC_TurnOnLights"))

(event Custom|PC_Day3
  ;[0 0 0] Console Command / Load and Play
  (Variables|Default|SetRunFromCheat true)
  (CallFunction|LoadDataLayersforDay3)
  (CallFunction|SetPortalAudio 2)
  (Utilities|FlowControl|Delay 7.77)
  (Development|ExecuteConsoleCommand "cheat ke * PC_TurnOnLights"))

(event Custom|PC_ZoneDL
  ;[0 0 0] Console Command / Load and Play
  (Variables|Default|SetRunFromCheat true)
  (CallFunction|LoadDataLayersforDay3)
  (CallFunction|SetPortalAudio 2)
  (Utilities|FlowControl|Delay 7.77)
  (Development|ExecuteConsoleCommand "cheat ke * PC_TurnOnLights"))

(event Custom|StartPreroll
  (params (Percentage :type Float%20%28double-precision%29))
  ;[0 0 0] Console Command / Load and Play
  (Variables|Cheats|SetCheat%20Day%20Percentage Percentage) ; Allows skipping ahead in the\nsequence for testing. A value\nbetween 0 - 1.0
  (if (Networking|IsServer)
    (Variables|Cheats|SetStarting%20Method "NewEnumerator1") ; Keeps some calendar logic from running
    (Variables|Default|SetRunFromCheat true)
    (CallFunction|%28CallOnServer%29Pre-LoadSequence)))

(event EventPre-LoadSequenceComplete
  ;[0 0 0] Console Command / Load and Play
  (|CallParentFunction)
  (if (Networking|IsServer)
    (if (Variables|Default|GetRunFromCheat)
      (CallFunction|LoadSequence) ; Play the sequence.
    )))

(event EventSequenceStopped(replicated)
  ; EndSequence Callback for RunnerRunner
  (Default|CallOnSequenceStopped))

(event Custom|LoadDataLayersForDay1
  (CallFunction|UnloadPistolCanaryCountdown)
  (Utilities|FlowControl|Delay 2.0)
  (bind _returnvalue (WorldPartition|GetDataLayerManager))
  (Utilities|FlowControl|Sequence
    (:then_0
      (DataLayers|SetDataLayerRuntimeState _returnvalue "/PistolCanary_Countdown/DataLayers/EDL_PistolCanary_Countdown.EDL_PistolCanary_Countdown" "Activated" false :!class /Script/Engine.DataLayerManager)
      (DataLayers|SetDataLayerRuntimeState _returnvalue "/PistolCanary_Countdown/DataLayers/DL_PistolCanary_Day1_to_End_S0.DL_PistolCanary_Day1_to_End_S0" "Activated" false :!class /Script/Engine.DataLayerManager)
      (DataLayers|SetDataLayerRuntimeState _returnvalue "/PistolCanary_Countdown/DataLayers/DL_PistolCanary_Day1_to_End_S3.DL_PistolCanary_Day1_to_End_S3" "Activated" false :!class /Script/Engine.DataLayerManager)
      (DataLayers|SetDataLayerRuntimeState _returnvalue "/PistolCanary_Countdown/DataLayers/DL_PistolCanary_Day1_to_MatchEnd.DL_PistolCanary_Day1_to_MatchEnd" "Activated" false :!class /Script/Engine.DataLayerManager)
      (DataLayers|SetDataLayerRuntimeState _returnvalue "/PistolCanary_Countdown/DataLayers/DL_PistolCanary_S0_to_End_S1.DL_PistolCanary_S0_to_End_S1" "Activated" false :!class /Script/Engine.DataLayerManager)
      (DataLayers|SetDataLayerRuntimeState _returnvalue "/PistolCanary_Countdown/DataLayers/DL_PistolCanary_S0_to_End_S3.DL_PistolCanary_S0_to_End_S3" "Activated" false :!class /Script/Engine.DataLayerManager))
    (:then_1
      (DataLayers|SetDataLayerRuntimeState _returnvalue 0 "Unloaded" false :!class /Script/Engine.DataLayerManager)
      (DataLayers|SetDataLayerRuntimeState _returnvalue 0 "Unloaded" false :!class /Script/Engine.DataLayerManager))
    (:then_2
      (Utilities|FlowControl|Delay 0.2)
      (CallFunction|PlayLevelSequenceandThenCleanUp "/PistolCanary_Countdown/Sequence/day/day100/PistolCanary_Countdown_day100.PistolCanary_Countdown_day100"))))

(event Custom|UnloadPistolCanaryCountdown
  (bind _returnvalue (WorldPartition|GetDataLayerManager))
  (DataLayers|SetDataLayerRuntimeState _returnvalue "/PistolCanary_Countdown/DataLayers/DL_PistolCanary_Countdown_EventParent.DL_PistolCanary_Countdown_EventParent" "Unloaded" true :!class /Script/Engine.DataLayerManager)
  (Utilities|FlowControl|Delay 0.0)
  (DataLayers|SetDataLayerRuntimeState _returnvalue "/PistolCanary_Countdown/DataLayers/DL_PistolCanary_Countdown_EventParent.DL_PistolCanary_Countdown_EventParent" "Activated" false :!class /Script/Engine.DataLayerManager))

(event Custom|PlayLevelSequenceAndThenCleanUp
  (params (LevelSequence :type Level%20Sequence%20Object%20Reference))
  (bind _returnvalue (Game|SpawnActorfromClass :Class "/Script/LevelSequence.ReplicatedLevelSequenceActor" :SpawnTransform (struct :Location "0, 0, 0" :Rotation "0, 0, 0" :Scale "1.000000,1.000000,1.000000") :CollisionHandlingOverride "AlwaysSpawn" :TransformScaleMethod "MultiplyWithRoot" :PlaybackSettings (struct :bAutoPlay true :LoopCount (struct :Value 0) :TickInterval "(TickIntervalSeconds=0.000000,EvaluationBudgetMicroseconds=0.000000,bTickWhenPaused=False,bAllowRounding=True)" :PlayRate 1.0 :StartTime 0.0 :bRandomStartTime false :bDisableMovementInput false :bDisableLookAtInput false :bHidePlayer false :bHideHud false :bDisableCameraCuts false :FinishCompletionStateOverride "None" :bPauseAtEnd false) :LevelSequenceAsset LevelSequence :bReplicatePlayback true))
  (Utilities|FlowControl|Delay 4.0)
  (Sequencer|Player|Stop (Variables|/Script/LevelSequence.LevelSequenceActor|GetSequencePlayer _returnvalue) :!class /Script/MovieScene.MovieSceneSequencePlayer)
  (Utilities|FlowControl|Delay 1.0)
  (Actor|DestroyActor _returnvalue))

(event Custom|LoadDataLayersForDay2
  (CallFunction|LoadDataLayersforDay1)
  (Utilities|FlowControl|Delay 2.0)
  (bind _returnvalue (WorldPartition|GetDataLayerManager))
  (DataLayers|SetDataLayerRuntimeState _returnvalue 0 "Unloaded" false :!class /Script/Engine.DataLayerManager)
  (DataLayers|SetDataLayerRuntimeState _returnvalue 0 "Unloaded" false :!class /Script/Engine.DataLayerManager)
  (DataLayers|SetDataLayerRuntimeState _returnvalue "/PistolCanary_Countdown/DataLayers/DL_PistolCanary_Day2_to_End_S1.DL_PistolCanary_Day2_to_End_S1" "Activated" false :!class /Script/Engine.DataLayerManager)
  (DataLayers|SetDataLayerRuntimeState _returnvalue "/PistolCanary_Countdown/DataLayers/DL_PistolCanary_Day2_to_MatchEnd.DL_PistolCanary_Day2_to_MatchEnd" "Activated" false :!class /Script/Engine.DataLayerManager)
  (DataLayers|SetDataLayerRuntimeState _returnvalue "/PistolCanary_Countdown/DataLayers/DL_PistolCanary_InnerSpire1.DL_PistolCanary_InnerSpire1" "Activated" false :!class /Script/Engine.DataLayerManager)
  (DataLayers|SetDataLayerRuntimeState _returnvalue "/PistolCanary_Countdown/DataLayers/DL_PistolCanary_InnerSpire2.DL_PistolCanary_InnerSpire2" "Activated" false :!class /Script/Engine.DataLayerManager)
  (DataLayers|SetDataLayerRuntimeState _returnvalue "/PistolCanary_Countdown/DataLayers/DL_PistolCanary_InnerSpire3.DL_PistolCanary_InnerSpire3" "Activated" false :!class /Script/Engine.DataLayerManager)
  (DataLayers|SetDataLayerRuntimeState _returnvalue "/PistolCanary_Countdown/DataLayers/DL_PistolCanary_InnerSpire4.DL_PistolCanary_InnerSpire4" "Activated" false :!class /Script/Engine.DataLayerManager)
  (Utilities|FlowControl|Delay 0.2)
  (CallFunction|PlayLevelSequenceandThenCleanUp "/PistolCanary_Countdown/Sequence/day/day200/PistolCanary_Countdown_day200.PistolCanary_Countdown_day200"))

(event Custom|LoadDataLayersForDay3
  (CallFunction|LoadDataLayersforDay2)
  (Utilities|FlowControl|Delay 2.0)
  (bind _returnvalue (WorldPartition|GetDataLayerManager))
  (DataLayers|SetDataLayerRuntimeState _returnvalue 0 "Unloaded" false :!class /Script/Engine.DataLayerManager)
  (DataLayers|SetDataLayerRuntimeState _returnvalue "/PistolCanary_Countdown/DataLayers/DL_PistolCanary_Day3_to_End_S0.DL_PistolCanary_Day3_to_End_S0" "Activated" false :!class /Script/Engine.DataLayerManager)
  (DataLayers|SetDataLayerRuntimeState _returnvalue "/PistolCanary_Countdown/DataLayers/DL_PistolCanary_Day3_to_End_S1.DL_PistolCanary_Day3_to_End_S1" "Activated" false :!class /Script/Engine.DataLayerManager)
  (Utilities|FlowControl|Delay 0.2)
  (CallFunction|PlayLevelSequenceandThenCleanUp "/PistolCanary_Countdown/Sequence/day/day300/PistolCanary_Countdown_day300.PistolCanary_Countdown_day300"))

(event Custom|SwitchToPreEventCastle
  (bind _returnvalue (WorldPartition|GetDataLayerManager))
  (DataLayers|SetDataLayerRuntimeState _returnvalue 0 "Unloaded" false :!class /Script/Engine.DataLayerManager)
  (DataLayers|SetDataLayerRuntimeState _returnvalue 0 "Unloaded" false :!class /Script/Engine.DataLayerManager)
  (Utilities|FlowControl|Delay 0.2)
  (DataLayers|SetDataLayerRuntimeState _returnvalue 0 "Activated" false :!class /Script/Engine.DataLayerManager)
  (DataLayers|SetDataLayerRuntimeState _returnvalue 0 "Activated" false :!class /Script/Engine.DataLayerManager)
  (DataLayers|SetDataLayerRuntimeState _returnvalue 0 "Activated" false :!class /Script/Engine.DataLayerManager)
  (DataLayers|SetDataLayerRuntimeState _returnvalue 0 "Activated" false :!class /Script/Engine.DataLayerManager)
  (DataLayers|SetDataLayerRuntimeState _returnvalue 0 "Activated" false :!class /Script/Engine.DataLayerManager)
  (DataLayers|SetDataLayerRuntimeState _returnvalue 0 "Activated" false :!class /Script/Engine.DataLayerManager)
  (DataLayers|SetDataLayerRuntimeState _returnvalue 0 "Activated" false :!class /Script/Engine.DataLayerManager)
  (DataLayers|SetDataLayerRuntimeState _returnvalue 0 "Activated" false :!class /Script/Engine.DataLayerManager))

(fn SetPortalAudio
  (params (EventDay :type Integer))
  (bind _returnvalue (Actor|GetActorOfClass "/PistolCanary_Audio/Gameplay/PistolCanary_AudioManager.PistolCanary_AudioManager_C"))
  (Class|PistolCanaryAudioManager|PlayProgressivePortalAudio _returnvalue EventDay))

(fn ConstructionScript
  (|CallParentFunction))
