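; BP_PistolCanary_SimpleKnockbackVolume is a sphere volume that launches Fortnite player pawns touching it.
; Collision is disabled at BeginPlay until EnableKnockbackCollision sets it to query-only; hit and begin-overlap both call KnockbackPlayer.
; The launch vector is the volume's normalized world location, not an offset to the player, scaled by DistanceMultiplier with Z clamped.
; End overlap clears a Timer_TeleportCheck handle that this graph never sets.
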(event EventBeginPlay
  (CallFunction|DisableCollision))

(event Custom|DisableCollision
  (Collision|SetCollisionEnabled (Variables|Default|GetSphere) "NoCollision"))

(event Custom|EnableKnockbackCollision
  (Collision|SetCollisionEnabled (Variables|Default|GetSphere) "QueryOnly"))

(event OnComponentHit(Sphere)
  (params (HitComponent :type Primitive%20Component%20Object%20Reference) (OtherActor :type Actor%20Object%20Reference) (OtherComp :type Primitive%20Component%20Object%20Reference) (NormalImpulse :type Vector :default "0, 0, 0") (Hit :type Hit%20Result%20Structure%20%28by%20ref%29))
  (bind _asfort_player_pawn (Utilities|Casting|CastToFortPlayerPawn OtherActor))
  (Fortnite|FortShippingLog self "Player hitting knockback volume")
  (CallFunction|KnockbackPlayer _asfort_player_pawn))

(event OnComponentBeginOverlap(Sphere)
  (params (OverlappedComponent :type Primitive%20Component%20Object%20Reference) (OtherActor :type Actor%20Object%20Reference) (OtherComp :type Primitive%20Component%20Object%20Reference) (OtherBodyIndex :type Integer) (bFromSweep :type Boolean) (SweepResult :type Hit%20Result%20Structure%20%28by%20ref%29))
  (bind _asfort_player_pawn (Utilities|Casting|CastToFortPlayerPawn OtherActor))
  (Variables|Default|SetPlayer _asfort_player_pawn)
  (Fortnite|FortShippingLog self "Player overlapping knockback volume")
  (CallFunction|KnockbackPlayer (Variables|Default|GetPlayer)))

(event Custom|KnockbackPlayer
  (params (Player :type Fort%20Pawn%20Object%20Reference))
  (bind (_returnvalue_x _returnvalue_y _returnvalue_z) (Math|Vector|BreakVector (Math|Vector|MakeVector (Variables|Default|GetDistanceMultiplier) (Variables|Default|GetDistanceMultiplier) (Variables|Default|GetDistanceMultiplier))))
  (bind (_returnvalue_x_2 _returnvalue_y_2 _returnvalue_z_2) (Math|Vector|BreakVector (* (Math|Vector|Normalize (Transformation|GetActorLocation)) (Math|Vector|MakeVector _returnvalue_x _returnvalue_y _returnvalue_z))))
  (Pawn|Character|LaunchCharacterJump Player (Math|Vector|MakeVector _returnvalue_x_2 _returnvalue_y_2 (Math|Float|Clamp(Float) _returnvalue_z_2 (Variables|Default|GetMinZClamp) (Variables|Default|GetMaxZClamp))) true true true true))

(event OnComponentEndOverlap(Sphere)
  (params (OverlappedComponent :type Primitive%20Component%20Object%20Reference) (OtherActor :type Actor%20Object%20Reference) (OtherComp :type Primitive%20Component%20Object%20Reference) (OtherBodyIndex :type Integer))
  (Utilities|Casting|CastToFortPlayerPawn OtherActor)
  (Utilities|Time|ClearandInvalidateTimerbyHandle (Variables|Default|GetTimer_TeleportCheck)))

(fn ConstructionScript)
